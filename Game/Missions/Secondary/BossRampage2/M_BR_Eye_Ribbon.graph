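# Unlit additive material for particle ribbons and beam trails, named as an eye ribbon effect for the BossRampage2 secondary mission.
# Particle color RGB runs through the shared MF_FX_EmissiveAdjustments function.
# Particle alpha is multiplied by the red channel of the T_Water_Top_Ramp gradient texture, masking the ribbon with no animation or custom UVs.
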Material /Game/Missions/Secondary/BossRampage2/M_BR_Eye_Ribbon
graph {
 properties {
  bUsedWithBeamTrails=true
  bUsedWithNiagaraRibbons=true
  blendMode=BLEND_Additive
  pixelDepthOffsetMode=PDOM_Legacy
  shadingModel=MSM_Unlit
  usedShadingModels=Unlit
 }
 nodes [
  n1 = ParticleColor()
  MaterialFunctionCall(Emissive=@n1.RGB) {materialFunction=&MF_FX_EmissiveAdjustments}
  Multiply(A=@n1.A, B=TextureSample() {texture=/Game/Global/Textures/Default/Masks_and_Gradients/T_Water_Top_Ramp}.R)
 ]
}
