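# Unlit additive mesh-particle material for a damage sphere effect in the BossRampage2 secondary mission.
# Fresnel rim plus a power-curved, inverted depth fade highlights edges and geometry intersections, then a short depth fade softens that mask.
# Particle color drives emissive through MF_FX_EmissiveAdjustments, and particle alpha scales the mask.
# Alpha times the inverted mask blends a separate value from one to slightly below, apparently for index-of-refraction distortion inside the rim.
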Material /Game/Missions/Secondary/BossRampage2/M_DamageSphere_Effect
graph {
 properties {
  bUsedWithMeshParticles=true
  blendMode=BLEND_Additive
  pixelDepthOffsetMode=PDOM_Legacy
  refractionMethod=RM_IndexOfRefraction
  shadingModel=MSM_Unlit
  usedShadingModels=Unlit
 }
 nodes [
  n1 = ParticleColor()
  n2 = Fresnel() {baseReflectFraction=0.02}
  n3 = LinearInterpolate(Alpha=Power(Base=DepthFade()) {constExponent=0.25}) {constA=1, constB=0}
  n4 = Clamp(None=DepthFade(Opacity=Add(A=@n2, B=@n3)) {fadeDistanceDefault=20})
  LinearInterpolate(Alpha=Multiply(A=@n1.A, B=OneMinus(None=@n4))) {constA=1, constB=0.8}
  MaterialFunctionCall(Emissive=@n1.RGB) {materialFunction=&MF_FX_EmissiveAdjustments}
  Multiply(A=@n1.A, B=@n4)
 ]
}
